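# M_ParentWheel_RLWH135 is the lit parent material for the RLWH135 cosmetic vehicle wheel on a skeletal mesh, covering rim and tire.
# Vertex color red and named-reroute masks blend rim and tire layers for color, normal, roughness, metallic and glow, with parameters for custom colors and surface values.
# Time drives a Rotator spinning rim masks, a Panner scrolling the tire mask, and a GoStrengh-scaled sine pulsing the animated patterns.
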Material /VehicleCosmetics/Wheels/RLWH135/New/Materials/M_ParentWheel_RLWH135
graph {
 properties {
  bAutomaticallySetUsageInEditor=false
  bUsedWithSkeletalMesh=true
  previewMesh=/VehicleCosmetics/Wheels/RLWH135/New/Meshes/SK_Wheel_RLWH135
  usedShadingModels=Default Lit
 }
 nodes [
  n1 = Time()
  n2 = Panner(Time=@n1) {speedX=-1}
  n3 = TextureSampleParameter2D(UVs=@n2) {parameterName=Tire_Mask_01, samplerType=SAMPLERTYPE_LinearColor, texture=/VehicleCosmetics/Wheels/RLWH135/New/Textures/T_Tire_RLWH135_A_Mask}
  n4 = OneMinus(None=@n3.R)
  LinearInterpolate(B=@n4, Alpha=NamedRerouteUsage())
  LinearInterpolate(
   A=TextureSampleParameter2D() {parameterName=Rim_N, samplerType=SAMPLERTYPE_Normal, texture=/VehicleCosmetics/Wheels/RLWH135/New/Textures/T_Rim_RLWH135_N}.RGB
   B=TextureSampleParameter2D() {parameterName=Tire_N, samplerType=SAMPLERTYPE_Normal, texture=/VehicleCosmetics/Wheels/RLWH135/New/Textures/T_Tire_RLWH135_N}.RGB
   Alpha=NamedRerouteUsage()
  )
  n5 = Multiply(A=@n1, B=Constant() {r=0.33})
  n6 = Multiply(A=Add(A=Sine(None=@n5), B=Constant() {r=1}), B=ScalarParameter() {defaultValue=1, parameterName=GoStrengh})
  n7 = Rotator(Time=@n1) {speed=-1}
  n8 = TextureSample(UVs=@n7) {texture=/VehicleCosmetics/Wheels/RLWH135/New/Textures/T_Rim_RLWH135_A_BC2}
  n9 = Add(A=Multiply(A=@n6, B=@n8.RGB), B=@n8.RGB)
  n10 = Constant() {r=1}
  n11 = TextureSample(UVs=@n2) {texture=/VehicleCosmetics/Wheels/RLWH135/New/Textures/T_Tire_RLWH135_A_BC2}
  n12 = Add(A=Multiply(A=@n11.RGB, B=@n6), B=@n11.RGB)
  n13 = TextureSample() {texture=/VehicleCosmetics/Wheels/RLWH135/New/Textures/T_Tire_RLWH135_A_BC1}
  LinearInterpolate(
   A=NamedRerouteDeclaration(None=Multiply(A=@n9, B=@n10)) {name=Emissive_Rim, nodeColor=<rgba 0.9219 1 0 1>}
   B=NamedRerouteDeclaration(None=Multiply(A=@n12, B=@n13.RGB)) {name=Emissive_Tire, nodeColor=<rgba 1 0 0.2812 1>}
   Alpha=NamedRerouteUsage()
  )
  n14 = VectorParameter() {parameterName=CustomColor_Rim_1, defaultValue=<rgba 0 0.0844 0.2831 1>}
  n15 = TextureSample(UVs=@n7) {samplerType=SAMPLERTYPE_LinearColor, texture=/VehicleCosmetics/Wheels/RLWH135/New/Textures/T_Rim_RLWH135_A_mask}
  n16 = TextureSample(UVs=@n7) {samplerType=SAMPLERTYPE_LinearColor, texture=/VehicleCosmetics/Wheels/RLWH135/New/Textures/T_Rim_RLWH135_A_D_fx_RGB}
  n17 = Subtract(A=OneMinus(None=@n15.R), B=@n16.R)
  n18 = Add(A=Multiply(A=@n6, B=@n17), B=@n17)
  n19 = Constant() {r=1}
  n20 = VectorParameter() {parameterName=CustomColor_Rim_2, defaultValue=<rgba 0.0222 1 1 1>}
  n21 = LinearInterpolate(A=LinearInterpolate(A=Constant(), B=@n14.RGB, Alpha=Multiply(A=@n18, B=@n19)), B=@n20.RGB, Alpha=@n15.R)
  n22 = Constant()
  n23 = NamedRerouteUsage()
  n24 = VertexColor()
  n25 = LinearInterpolate(A=Multiply(A=LinearInterpolate(A=@n21, B=@n22, Alpha=@n23), B=Constant() {r=0.4}), B=Constant(), Alpha=@n24.R)
  n26 = Constant()
  n27 = NamedRerouteUsage()
  n28 = TextureSampleParameter2D() {parameterName=Tire_Mas_02, texture=&T_Tire_RLWH135_A_BC1}
  n29 = Clamp(None=ConstantBiasScale(None=@n28.R) {bias=-0.2, scale=5})
  n30 = ConstantBiasScale(None=OneMinus(None=@n29)) {bias=-0.1}
  n31 = Constant()
  n32 = Multiply(A=@n3.R, B=@n6)
  n33 = LinearInterpolate(A=LinearInterpolate(A=Constant(), B=NamedRerouteUsage(), Alpha=Add(A=@n32, B=@n3.R)), B=NamedRerouteUsage(), Alpha=@n4)
  n34 = NamedRerouteUsage()
  n35 = LinearInterpolate(B=LinearInterpolate(A=LinearInterpolate(A=@n31, B=@n33, Alpha=@n29), B=@n33, Alpha=@n28.RGB), Alpha=@n34)
  LinearInterpolate(A=@n25, B=Max(A=LinearInterpolate(A=@n26, B=@n27, Alpha=@n30), B=@n35), Alpha=@n34)
  n36 = LinearInterpolate(A=Constant() {r=0.05}, B=ScalarParameter() {defaultValue=1, parameterName=Roughness_Picture_Tire}, Alpha=@n29)
  n37 = ScalarParameter() {defaultValue=0.3, parameterName=Roughness_Tire}
  n38 = ConstantBiasScale(None=OneMinus(None=@n29)) {bias=-0.1, scale=1}
  n39 = NamedRerouteUsage()
  LinearInterpolate(A=NamedRerouteUsage(), B=LinearInterpolate(B=LinearInterpolate(A=@n36, B=@n37, Alpha=@n38), Alpha=@n39), Alpha=@n39)
  n40 = LinearInterpolate(A=Constant() {r=0.05}, B=ScalarParameter() {parameterName=Metallic_Picture_Tire}, Alpha=@n29)
  n41 = ScalarParameter() {parameterName=Metallic_Tire}
  n42 = NamedRerouteUsage()
  LinearInterpolate(A=NamedRerouteUsage(), B=LinearInterpolate(B=LinearInterpolate(A=@n40, B=@n41), Alpha=@n42), Alpha=@n42)
  Multiply(A=LinearInterpolate(A=@n25, B=@n35, Alpha=NamedRerouteUsage()), B=ScalarParameter() {defaultValue=2, parameterName=Emissive_Glow})
  n43 = TextureSample() {samplerType=SAMPLERTYPE_LinearColor, texture=&T_Rim_RLWH135_A_D_fx_RGB}
  Multiply(A=@n43.R)
  Multiply(A=@n43.G, B=Constant() {r=20})
  Multiply(A=@n16.R)
  n44 = ScalarParameter() {parameterName=Roughness_Picture_01}
  n45 = ScalarParameter() {defaultValue=0.4, parameterName=Roughness_Rim}
  n46 = NamedRerouteUsage()
  NamedRerouteDeclaration(None=LinearInterpolate(A=Constant(), B=LinearInterpolate(A=@n44, B=@n45, Alpha=@n43.R), Alpha=OneMinus(None=@n46))) {name=Roughness_rim, nodeColor=<rgba 1 0.5859 0 1>}
  NamedRerouteDeclaration(None=@n43.R) {name=Rim_M, nodeColor=<rgba 0.2656 0 1 1>}
  n47 = ScalarParameter() {parameterName=Metallic_Picture_01}
  n48 = ScalarParameter() {defaultValue=0.054, parameterName=Metallic_Rim}
  n49 = NamedRerouteUsage()
  n50 = NamedRerouteUsage()
  NamedRerouteDeclaration(None=LinearInterpolate(A=Constant(), B=LinearInterpolate(A=@n47, B=@n48, Alpha=@n49), Alpha=OneMinus(None=@n50))) {name=Metallic_Rim, nodeColor=<rgba 1 0.5859 0 1>}
  NamedRerouteDeclaration(None=@n14.RGB) {name=CustomColor_Rim_1, nodeColor=<rgba 0.5938 0 1 1>}
  NamedRerouteDeclaration(None=@n20.RGB) {name=CustomColor_Rim_2, nodeColor=<rgba 0.5938 0 1 1>}
  NamedRerouteDeclaration(None=VectorParameter() {parameterName=CustomColor_Tire_Rim, defaultValue=<rgba 0.074 0.074 0.074 1>}.RGB) {name=CustomColor_Tire_3, nodeColor=<rgba 0.5938 0 1 1>}
  NamedRerouteDeclaration(None=@n24.R) {name=VC_RC, nodeColor=<rgba 0 0.3672 1 1>}
  NamedRerouteDeclaration(None=@n30) {name=TireMask01, nodeColor=<rgba 1 0 0.2109 1>}
  TextureSample(UVs=@n2) {texture=/VehicleCosmetics/Wheels/RLWH135/New/Textures/T_Tire_RLWH135_BC}
 ]
}
